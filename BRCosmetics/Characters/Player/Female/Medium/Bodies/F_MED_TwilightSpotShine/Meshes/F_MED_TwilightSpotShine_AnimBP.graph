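# Cosmetic AnimBP for the F_MED_TwilightSpotShine female body, adding secondary motion to its chain and collar.
# The attached parent's copied pose runs through a Control Rig, a back-collar offset, and weighted multi-bone constraints interleaved with spring-driven Anim Dynamics.
# Constraints anchor chain and collar bones to body or neighboring chain bones.
# The event graph sets front-end alpha and simulation-space variables, which the AnimGraph shown never reads.
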AnimBlueprint /BRCosmetics/Characters/Player/Female/Medium/Bodies/F_MED_TwilightSpotShine/Meshes/F_MED_TwilightSpotShine_AnimBP
target_skeleton=&/F_MED_TwilightSpotShine_Skeleton
graphs [
 AnimGraph {nodes [
   n1 = Control Rig(Source=Copy Pose From Mesh() {bUseAttachedParent=true}) {
    alphaBoolBlend {blendOption=Linear}
    controlRigAssetReference {blueprintRigClass=&/F_MED_TwilightSpotShine_CtrlRig.F_MED_TwilightSpotShine_CtrlRig_C}
   }
   n2 = "Transform (Modify) Bone - Bone: dyn_collar_bk"(ComponentPose=Local To Component(LocalPose=@n1), Alpha=1, Translation="0,-1,0", Rotation="0,0,0", Scale="1,1,1") {translationMode=BMM_Additive, alphaBoolBlend {blendOption=Linear}, boneToModify {boneName=dyn_collar_bk}, translation {y=-1}}
   n3 = "Constraint - dyn_chain_05 (PPPP)"!(
    ComponentPose=@n2
    Alpha=1
    ConstraintWeights_0=0.0197
    ConstraintWeights_1:float=0.005
    ConstraintWeights_2:float=0.9212
    ConstraintWeights_3:float=0.0541
   ) -> (Pose) {
    alphaBoolBlend {blendOption=Linear}
    boneToModify {boneName=dyn_chain_05}
    constraintSetup [{targetBone {boneName=spine_02}}, {targetBone {boneName=spine_01}}, {targetBone {boneName=pelvis}}, {targetBone {boneName=deform_glute_l}}]
    constraintWeights [1,1,1,1]
   }
   n4 = Anim Dynamics - Bone: dyn_chain_01!(SimulationSpace:AnimPhysSimSpaceType Enum=Component, ComponentPose=@n3, Alpha=1, ExternalForce="0,0,0") -> (Pose) {
    angularDampingOverride=0.89
    angularSpringConstant=100
    bAngularSpring=true
    bGravityOverrideInSimSpace=true
    bOverrideAngularDamping=true
    bUseGravityOverride=true
    linearDampingOverride=0.9
    linearSpringConstant=150
    numSolverIterationsPreUpdate=3
    simulationSpace=BoneRelative
    alphaBoolBlend {blendOption=Linear}
    boundBone {boneName=dyn_chain_01}
    componentAppliedLinearAccClamp=<1000 1000 1000>
    componentLinearAccScale=<0.5 0.5 0.5>
    componentLinearVelScale=<0.5 0.5 0.5>
    gravityOverride {z=-980}
    physicsBodyDefinitions [{
      boundBone {boneName=dyn_chain_01}
      boxExtents=<15 15 15>
      constraintSetup {angularLimitsMax=<0 0 0>, angularLimitsMin=<0 -30 -10>, angularTarget=<1 0 0>}
      localJointOffset=<2 0 0>
     }]
    relativeSpaceBone {boneName=root}
   }
   n5 = Anim Dynamics - Bone: dyn_chain_02!(SimulationSpace:AnimPhysSimSpaceType Enum=BoneRelative, ComponentPose=@n4, Alpha=1, ExternalForce="0,0,0") -> (Pose) {
    angularDampingOverride=0.89
    angularSpringConstant=100
    bAngularSpring=true
    bGravityOverrideInSimSpace=true
    bOverrideAngularDamping=true
    bUseGravityOverride=true
    linearDampingOverride=0.9
    linearSpringConstant=150
    numSolverIterationsPreUpdate=3
    simulationSpace=BoneRelative
    alphaBoolBlend {blendOption=Linear}
    boundBone {boneName=dyn_chain_02}
    componentAppliedLinearAccClamp=<1000 1000 1000>
    componentLinearAccScale=<0.5 0.5 0.5>
    componentLinearVelScale=<0.5 0.5 0.5>
    gravityOverride {z=-980}
    physicsBodyDefinitions [{
      boundBone {boneName=dyn_chain_02}
      boxExtents=<15 15 15>
      constraintSetup {angularLimitsMax=<0 0 0>, angularLimitsMin=<0 -30 -10>, angularTarget=<1 0 0>}
      localJointOffset=<2 0 0>
     }]
    relativeSpaceBone {boneName=root}
   }
   n6 = Anim Dynamics - Bone: dyn_chain_03!(SimulationSpace:AnimPhysSimSpaceType Enum=BoneRelative, ComponentPose=@n5, Alpha=1, ExternalForce="0,0,0") -> (Pose) {
    angularDampingOverride=0.89
    angularSpringConstant=100
    bAngularSpring=true
    bGravityOverrideInSimSpace=true
    bOverrideAngularDamping=true
    bUseGravityOverride=true
    linearDampingOverride=0.9
    linearSpringConstant=150
    numSolverIterationsPreUpdate=3
    simulationSpace=BoneRelative
    alphaBoolBlend {blendOption=Linear}
    boundBone {boneName=dyn_chain_03}
    componentAppliedLinearAccClamp=<1000 1000 1000>
    componentLinearAccScale=<0.5 0.5 0.5>
    componentLinearVelScale=<0.5 0.5 0.5>
    gravityOverride {z=-980}
    physicsBodyDefinitions [{
      boundBone {boneName=dyn_chain_03}
      boxExtents=<15 15 15>
      constraintSetup {angularLimitsMax=<0 0 0>, angularLimitsMin=<0 -20 -10>, angularTarget=<1 0 0>}
      localJointOffset=<2 0 0>
     }]
    relativeSpaceBone {boneName=root}
   }
   n7 = Anim Dynamics - Bone: dyn_chain_04!(SimulationSpace:AnimPhysSimSpaceType Enum=BoneRelative, ComponentPose=@n6, Alpha=1, ExternalForce="0,0,0") -> (Pose) {
    angularDampingOverride=0.89
    angularSpringConstant=100
    bAngularSpring=true
    bGravityOverrideInSimSpace=true
    bOverrideAngularDamping=true
    bUseGravityOverride=true
    linearDampingOverride=0.9
    linearSpringConstant=150
    numSolverIterationsPreUpdate=3
    simulationSpace=BoneRelative
    alphaBoolBlend {blendOption=Linear}
    boundBone {boneName=dyn_chain_04}
    componentAppliedLinearAccClamp=<10000 10000 10000>
    componentLinearAccScale=<0.5 0.5 0.5>
    componentLinearVelScale=<0.5 0.5 0.5>
    gravityOverride {z=-980}
    physicsBodyDefinitions [{
      boundBone {boneName=dyn_chain_04}
      boxExtents=<15 15 15>
      constraintSetup {angularLimitsMax=<0 0 0>, angularLimitsMin=<0 -20 -10>, angularTarget=<1 0 0>}
      localJointOffset=<2 0 0>
     }]
    relativeSpaceBone {boneName=root}
   }
   n8 = Anim Dynamics - Bone: dyn_chain_05!(SimulationSpace:AnimPhysSimSpaceType Enum=BoneRelative, ComponentPose=@n7, Alpha=1, ExternalForce="0,0,0") -> (Pose) {
    angularDampingOverride=0.89
    angularSpringConstant=100
    bAngularSpring=true
    bGravityOverrideInSimSpace=true
    bOverrideAngularDamping=true
    bUseGravityOverride=true
    linearDampingOverride=0.9
    linearSpringConstant=150
    numSolverIterationsPreUpdate=3
    simulationSpace=BoneRelative
    alphaBoolBlend {blendOption=Linear}
    boundBone {boneName=dyn_chain_05}
    componentAppliedLinearAccClamp=<1000 1000 1000>
    componentLinearAccScale=<0.5 0.5 0.5>
    componentLinearVelScale=<0.5 0.5 0.5>
    gravityOverride {z=-980}
    physicsBodyDefinitions [{
      boundBone {boneName=dyn_chain_05}
      boxExtents=<15 15 15>
      constraintSetup {angularLimitsMax=<0 0 0>, angularLimitsMin=<0 -10 -10>, angularTarget=<1 0 0>}
      localJointOffset=<-2 0 0>
     }]
    relativeSpaceBone {boneName=root}
   }
   n9 = Anim Dynamics - Bone: dyn_chain_06!(SimulationSpace:AnimPhysSimSpaceType Enum=BoneRelative, ComponentPose=@n8, Alpha=1, ExternalForce="0,0,0") -> (Pose) {
    angularDampingOverride=0.89
    angularSpringConstant=100
    bAngularSpring=true
    bGravityOverrideInSimSpace=true
    bOverrideAngularDamping=true
    bUseGravityOverride=true
    linearDampingOverride=0.9
    linearSpringConstant=150
    numSolverIterationsPreUpdate=3
    simulationSpace=BoneRelative
    alphaBoolBlend {blendOption=Linear}
    boundBone {boneName=dyn_chain_06}
    componentAppliedLinearAccClamp=<1000 1000 1000>
    componentLinearAccScale=<0.5 0.5 0.5>
    componentLinearVelScale=<0.5 0.5 0.5>
    gravityOverride {z=-980}
    physicsBodyDefinitions [{
      boundBone {boneName=dyn_chain_06}
      boxExtents=<15 15 15>
      constraintSetup {angularLimitsMax=<0 0 0>, angularLimitsMin=<0 -10 -10>, angularTarget=<1 0 0>}
      localJointOffset=<-2 0 0>
     }]
    relativeSpaceBone {boneName=root}
   }
   n10 = Anim Dynamics - Bone: dyn_chain_07!(SimulationSpace:AnimPhysSimSpaceType Enum=BoneRelative, ComponentPose=@n9, Alpha=1, ExternalForce="0,0,0") -> (Pose) {
    angularDampingOverride=0.89
    angularSpringConstant=100
    bAngularSpring=true
    bGravityOverrideInSimSpace=true
    bOverrideAngularDamping=true
    bUseGravityOverride=true
    linearDampingOverride=0.9
    linearSpringConstant=150
    numSolverIterationsPreUpdate=3
    simulationSpace=BoneRelative
    alphaBoolBlend {blendOption=Linear}
    boundBone {boneName=dyn_chain_07}
    componentAppliedLinearAccClamp=<1000 1000 1000>
    componentLinearAccScale=<0.5 0.5 0.5>
    componentLinearVelScale=<0.5 0.5 0.5>
    gravityOverride {z=-980}
    physicsBodyDefinitions [{
      boundBone {boneName=dyn_chain_07}
      boxExtents=<15 15 15>
      constraintSetup {angularLimitsMax=<0 0 0>, angularLimitsMin=<0 -20 -10>, angularTarget=<1 0 0>}
      localJointOffset=<-2 0 0>
     }]
    relativeSpaceBone {boneName=root}
   }
   n11 = Anim Dynamics - Bone: dyn_chain_08!(SimulationSpace:AnimPhysSimSpaceType Enum=BoneRelative, ComponentPose=@n10, Alpha=1, ExternalForce="0,0,0") -> (Pose) {
    angularDampingOverride=0.89
    angularSpringConstant=100
    bAngularSpring=true
    bGravityOverrideInSimSpace=true
    bOverrideAngularDamping=true
    bUseGravityOverride=true
    linearDampingOverride=0.9
    linearSpringConstant=150
    numSolverIterationsPreUpdate=3
    simulationSpace=BoneRelative
    alphaBoolBlend {blendOption=Linear}
    boundBone {boneName=dyn_chain_08}
    componentAppliedLinearAccClamp=<10000 10000 10000>
    componentLinearAccScale=<0.5 0.5 0.5>
    componentLinearVelScale=<0.5 0.5 0.5>
    gravityOverride {z=-980}
    physicsBodyDefinitions [{
      boundBone {boneName=dyn_chain_08}
      boxExtents=<15 15 15>
      constraintSetup {angularLimitsMax=<0 0 0>, angularLimitsMin=<0 -20 -10>, angularTarget=<1 0 0>}
      localJointOffset=<2 0 0>
     }]
    relativeSpaceBone {boneName=root}
   }
   n12 = Anim Dynamics - Bone: dyn_chain_c!(SimulationSpace:AnimPhysSimSpaceType Enum=BoneRelative, ComponentPose=@n11, Alpha=1, ExternalForce="0,0,0") -> (Pose) {
    angularDampingOverride=0.89
    angularSpringConstant=100
    bAngularSpring=true
    bGravityOverrideInSimSpace=true
    bOverrideAngularDamping=true
    bUseGravityOverride=true
    linearDampingOverride=0.9
    linearSpringConstant=150
    numSolverIterationsPreUpdate=3
    simulationSpace=BoneRelative
    alphaBoolBlend {blendOption=Linear}
    boundBone {boneName=dyn_chain_c}
    componentAppliedLinearAccClamp=<10000 10000 10000>
    componentLinearAccScale=<0.5 0.5 0.5>
    componentLinearVelScale=<0.5 0.5 0.5>
    gravityOverride {z=-980}
    physicsBodyDefinitions [{
      boundBone {boneName=dyn_chain_c}
      boxExtents=<17 17 17>
      constraintSetup {angularLimitsMax=<0 0 5>, angularLimitsMin=<0 -30 -5>, angularTarget=<1 0 0>}
      localJointOffset=<2 0 0>
     }]
    relativeSpaceBone {boneName=root}
   }
   n13 = "Constraint - dyn_chain_c (PPP)"!(ComponentPose=@n12, Alpha=1, ConstraintWeights_0=0.5, ConstraintWeights_1:float=0.5, ConstraintWeights_2:float=0) -> (Pose) {
    alphaBoolBlend {blendOption=Linear}
    boneToModify {boneName=dyn_chain_c}
    constraintSetup [{targetBone {boneName=dyn_chain_04}}, {targetBone {boneName=dyn_chain_08}}, {targetBone {boneName=thigh_l}}]
    constraintWeights [1,1,1]
   }
   n14 = "Constraint - dyn_collar_r (PPPPPPP)"!(
    ComponentPose=@n13
    Alpha=1
    ConstraintWeights_0=0.6476
    ConstraintWeights_1:float=0.1025
    ConstraintWeights_2:float=0.0918
    ConstraintWeights_3:float=0.0744
    ConstraintWeights_4:float=0.0555
    ConstraintWeights_5:float=0.0188
    ConstraintWeights_6:float=0.0093
   ) -> (Pose) {
    alphaBoolBlend {blendOption=Linear}
    boneToModify {boneName=dyn_collar_r}
    constraintSetup [
     {targetBone {boneName=spine_05}}
     {targetBone {boneName=spine_04}}
     {targetBone {boneName=neck_01}}
     {targetBone {boneName=neck_02}}
     {targetBone {boneName=clavicle_l}}
     {targetBone {boneName=head}}
     {targetBone {boneName=spine_03}}
    ]
    constraintWeights [1,1,1,1,1,1,1]
   }
   n15 = "Constraint - dyn_collar_l (PPPPPPP)"!(
    ComponentPose=@n14
    Alpha=1
    ConstraintWeights_0=0.8027
    ConstraintWeights_1:float=0.0696
    ConstraintWeights_2:float=0.0468
    ConstraintWeights_3:float=0.0427
    ConstraintWeights_4:float=0.0272
    ConstraintWeights_5:float=0.0063
    ConstraintWeights_6:float=0.0048
   ) -> (Pose) {
    alphaBoolBlend {blendOption=Linear}
    boneToModify {boneName=dyn_collar_l}
    constraintSetup [
     {targetBone {boneName=spine_05}}
     {targetBone {boneName=spine_04}}
     {targetBone {boneName=clavicle_r}}
     {targetBone {boneName=neck_01}}
     {targetBone {boneName=neck_02}}
     {targetBone {boneName=spine_03}}
     {targetBone {boneName=head}}
    ]
    constraintWeights [1,1,1,1,1,1,1]
   }
   n16 = Anim Dynamics - Bone: dyn_collar_r(ComponentPose=@n15, Alpha=1, ExternalForce="0,0,0") {
    angularDampingOverride=1
    angularSpringConstant=20000
    bAngularSpring=true
    bOverrideAngularDamping=true
    bUseGravityOverride=true
    gravityScale=0.35
    simulationSpace=BoneRelative
    alphaBoolBlend {blendOption=Linear}
    boundBone {boneName=dyn_collar_r}
    componentLinearAccScale=<0.5 0.5 0.5>
    componentLinearVelScale=<0.5 0.5 0.5>
    gravityOverride {z=-980}
    physicsBodyDefinitions [{
      boundBone {boneName=dyn_collar_r}
      boxExtents=<20 30 30>
      constraintSetup {angularLimitsMax=<5 5 15>, angularLimitsMin=<-5 -5 -30>, angularTarget=<1 0 0>}
      localJointOffset=<4 0 0>
     }]
    relativeSpaceBone {boneName=spine_05}
   }
   n17 = Anim Dynamics - Bone: dyn_collar_l(ComponentPose=@n16, Alpha=1, ExternalForce="0,0,0") {
    angularDampingOverride=1
    angularSpringConstant=20000
    bAngularSpring=true
    bOverrideAngularDamping=true
    bUseGravityOverride=true
    gravityScale=0.35
    simulationSpace=BoneRelative
    alphaBoolBlend {blendOption=Linear}
    boundBone {boneName=dyn_collar_l}
    componentLinearAccScale=<0.5 0.5 0.5>
    componentLinearVelScale=<0.5 0.5 0.5>
    gravityOverride {z=-980}
    physicsBodyDefinitions [{
      boundBone {boneName=dyn_collar_l}
      boxExtents=<20 30 30>
      constraintSetup {angularLimitsMax=<5 5 30>, angularLimitsMin=<-5 -5 -15>, angularTarget=<1 0 0>}
      localJointOffset=<4 0 0>
     }]
    relativeSpaceBone {boneName=spine_05}
   }
   Output Pose(Result=Component To Local(ComponentPose=@n17)) {layerGroup=None}
  ]}
 EventGraph {nodes [
   Event BlueprintUpdateAnimation()
   > n1 = Branch(Condition=Get bIsInFrontEnd())
   > Set FrontendAlpha(FrontendAlpha:double=0.1) -> (Output_Get:double)
   > Set Simulation Space(Simulation Space:AnimPhysSimSpaceType Enum=World) -> (Output_Get:AnimPhysSimSpaceType Enum)
   > Set LayeredFE_DynAlpha(LayeredFE_DynAlpha:double=0.1) -> (Output_Get:double)
   Set FrontendAlpha(execute=@n1.else, FrontendAlpha=1)
   > Set Simulation Space(Simulation Space=BoneRelative)
   > Set LayeredFE_DynAlpha(LayeredFE_DynAlpha=0)
   Event OnCharacterCustomizationLoaded()
  ]}
]
variables [
 FrontendAlpha {type=number}
 LayeredFE_DynAlpha {type=number}
 Simulation Space {type=AnimPhysSimSpaceType}
]
