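# Physics asset for the female medium TwilightSpotShine cosmetic body, covering a three-bone dynamic front hair chain plus the head and upper spine.
# Each of the five bones has one capsule, with large head and spine_05 shapes beside small hair shapes.
# Constraints chain hair to head to spine_05, leaving the hair root and head angularly free while the lower hair links have moderate swing and very tight twist limits.
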PhysicsAsset /BRCosmetics/Characters/Player/Female/Medium/Bodies/F_MED_TwilightSpotShine/Meshes/Parts/F_MED_TwilightSpotShine_Physics
bodies {
 dyn_hairFront_01 [{shapeType=Capsule, length=0.8153, radius=0.7213, center=<0.4983 0.745 -0.3056>, rotation=<rot 3.1575 -90 19.9528>}]
 dyn_hairFront_02 [{shapeType=Capsule, length=2.4668, radius=1.3798, center=<1.4649 -0.2972 0.2197>, rotation=<rot 0 91.9494 -86.7675>}]
 dyn_hairFront_03 [{shapeType=Capsule, length=2.9006, radius=1.5569, center=<1.9348 0.0067 -0.1746>, rotation=<rot 0 91.935 -89.6691>}]
 head [{shapeType=Capsule, length=10.0505, radius=6.0598, center=<3.1811 1.7982 0.347>, rotation=<rot 1.0348 90.9166 -88.5218>}]
 spine_05 [{shapeType=Capsule, length=3, radius=10.005, shapeName=spine_05_capsule, center=<4.138 -1.6387 0>, rotation=<rot 0 -90 0>}]
}
constraints [
 {
  bone1Name=dyn_hairFront_01
  bone2Name=head
  swing1LimitDegrees=45
  swing1Motion=Free
  swing2LimitDegrees=45
  swing2Motion=Free
  twistLimitDegrees=45
  twistMotion=Free
 }
 {bone1Name=dyn_hairFront_02, bone2Name=dyn_hairFront_01, swing1LimitDegrees=35, swing2LimitDegrees=35, twistLimitDegrees=1}
 {bone1Name=dyn_hairFront_03, bone2Name=dyn_hairFront_02, swing1LimitDegrees=35, swing2LimitDegrees=35, twistLimitDegrees=1}
 {
  bone1Name=head
  bone2Name=spine_05
  swing1LimitDegrees=45
  swing1Motion=Free
  swing2LimitDegrees=45
  swing2Motion=Free
  twistLimitDegrees=45
  twistMotion=Free
 }
]
