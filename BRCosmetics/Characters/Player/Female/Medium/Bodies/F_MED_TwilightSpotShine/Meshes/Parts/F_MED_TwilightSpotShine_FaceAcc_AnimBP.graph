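# Post-attach animation blueprint for the TwilightSpotShine female character's face accessory mesh, adding secondary motion to its hair bones.
# It copies the pose from the attached parent mesh, constrains the left temple hair bone to the head and slightly offsets the left hair base.
# Three spring-driven Anim Dynamics nodes simulate right, left and temple hair, then a RigidBody node runs the character's physics asset relative to spine_05.
# The event graph contains no logic.
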AnimBlueprint /BRCosmetics/Characters/Player/Female/Medium/Bodies/F_MED_TwilightSpotShine/Meshes/Parts/F_MED_TwilightSpotShine_FaceAcc_AnimBP
target_skeleton=&/F_MED_TwilightSpotShine_FaceAcc_Skeleton
graphs [
 AnimGraph {nodes [
   n1 = "Constraint - dyn_hairTemple_l (P)"(ComponentPose=Local To Component(LocalPose=Copy Pose From Mesh() {bUseAttachedParent=true}), Alpha=1, ConstraintWeights_0=1) {
    alphaBoolBlend {blendOption=Linear}
    boneToModify {boneName=dyn_hairTemple_l}
    constraintSetup [{targetBone {boneName=head}}]
    constraintWeights [1]
   }
   n2 = "Transform (Modify) Bone - Bone: base_hair_l"(ComponentPose=@n1, Alpha=1, Translation="0,1,0", Rotation="0,0,0", Scale="1,1,1") {translationMode=BMM_Additive, alphaBoolBlend {blendOption=Linear}, boneToModify {boneName=base_hair_l}, translation {y=1}}
   n3 = Anim Dynamics - Bone: dyn_hair_r(ComponentPose=@n2, Alpha=1, ExternalForce="0,0,0") {
    angularDampingOverride=0.9999
    angularSpringConstant=250
    bAngularSpring=true
    bGravityOverrideInSimSpace=true
    bOverrideAngularDamping=true
    bUseGravityOverride=true
    bUsePlanarLimit=false
    bUseSphericalLimits=true
    gravityScale=0.35
    simulationSpace=RootRelative
    alphaBoolBlend {blendOption=Linear}
    boundBone {boneName=dyn_hair_r}
    componentAppliedLinearAccClamp=<70 70 70>
    componentLinearVelScale=<300 300 300>
    gravityOverride {z=-980}
    physicsBodyDefinitions [{
      sphereCollisionRadius=3
      boundBone {boneName=dyn_hair_r}
      boxExtents=<10 10 10>
      constraintSetup {angularLimitsMax=<5 20 5>, angularLimitsMin=<-5 -45 -40>, angularTarget=<1 0 0>}
      localJointOffset=<1 0 0>
     }]
    relativeSpaceBone {boneName=spine_05}
    sphericalLimits [{limitRadius=2.5, drivingBone {boneName=neck_02}, sphereLocalOffset=<0 0 0>}]
   }
   n4 = Anim Dynamics - Bone: dyn_hair_l(ComponentPose=@n3, Alpha=1, ExternalForce="0,0,0") {
    angularDampingOverride=1
    angularSpringConstant=250
    bAngularSpring=true
    bGravityOverrideInSimSpace=true
    bOverrideAngularDamping=true
    bUsePlanarLimit=false
    gravityScale=0.35
    alphaBoolBlend {blendOption=Linear}
    boundBone {boneName=dyn_hair_l}
    componentAppliedLinearAccClamp=<0 0 0>
    gravityOverride {z=-980}
    physicsBodyDefinitions [{
      sphereCollisionRadius=3
      boundBone {boneName=dyn_hair_l}
      boxExtents=<10 10 10>
      constraintSetup {angularLimitsMax=<5 10 20>, angularLimitsMin=<-5 -15 -65>, angularTarget=<0 0 0>}
      localJointOffset=<2 0 0>
     }]
    sphericalLimits [{limitRadius=5, drivingBone {boneName=neck_02}, sphereLocalOffset=<0 0 0>}]
   }
   n5 = Anim Dynamics - Bone: dyn_hairTemple_l(ComponentPose=@n4, Alpha=1, ExternalForce="0,0,0") {
    angularDampingOverride=0.9999
    angularSpringConstant=250
    bAngularSpring=true
    bGravityOverrideInSimSpace=true
    bOverrideAngularDamping=true
    bUseGravityOverride=true
    bUsePlanarLimit=false
    bUseSphericalLimits=true
    gravityScale=0.35
    simulationSpace=RootRelative
    alphaBoolBlend {blendOption=Linear}
    boundBone {boneName=dyn_hairTemple_l}
    componentAppliedLinearAccClamp=<70 70 70>
    componentLinearVelScale=<300 300 300>
    gravityOverride {z=-980}
    physicsBodyDefinitions [{
      sphereCollisionRadius=3
      boundBone {boneName=dyn_hairTemple_l}
      boxExtents=<10 10 10>
      constraintSetup {angularLimitsMax=<10 15 10>, angularLimitsMin=<-10 -15 -40>, angularTarget=<1 0 0>}
      localJointOffset=<2 0 0>
     }]
    relativeSpaceBone {boneName=spine_05}
    sphericalLimits [{limitRadius=2.5, drivingBone {boneName=neck_02}, sphereLocalOffset=<0 0 0>}]
   }
   n6 = RigidBody(ComponentPose=@n5) {
    bOverrideWorldGravity=true
    overridePhysicsAsset=&/F_MED_TwilightSpotShine_Physics
    simulationSpace=BaseBoneSpace
    alphaBoolBlend {blendOption=Linear}
    baseBoneRef {boneName=spine_05}
    componentAppliedLinearAccClamp=<1000 1000 1000>
    componentLinearAccScale=<0.3 0.4 0.4>
    componentLinearVelScale=<0.3 0.4 0.4>
    overrideWorldGravity {z=-980}
   }
   Output Pose(Result=Component To Local(ComponentPose=@n6)) {layerGroup=None}
  ]}
 EventGraph {nodes [Event BlueprintUpdateAnimation()]}
]
variables []
